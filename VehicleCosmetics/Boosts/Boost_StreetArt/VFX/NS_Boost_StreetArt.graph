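# Niagara system for the Street Art vehicle boost cosmetic, trailing paint-variant-colored dots and ribbons from the car's boost sockets.
# Two sprite emitters partition particles across filtered skeletal mesh sockets, one longer-lived stream and one short-lived wider cone, both inheriting source movement.
# The ribbon emitter spawns per port, layering five renderers with curve-driven widths, three painted color layers and per-partition sign-flipped material parameters.
# Particle lifetimes switch on GameModeID.
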NiagaraSystem /VehicleCosmetics/Boosts/Boost_StreetArt/VFX/NS_Boost_StreetArt
dependencies {
 usedDataInterfaces [&NiagaraDataInterfaceColorCurve, &NiagaraDataInterfaceCurve, &NiagaraDataInterfaceSkeletalMesh]
 usedDynamicInputs []
 usedModules [
  &self:SystemSpawnScript.SystemScriptSource.SystemScriptGraph.NiagaraNodeAssignment_0.SetVariables_#1
  &SystemState
 ]
}
emitters {
 NE_Dots {
  emitterUpdateScript [
   &EmitterState()
   &VC_FilteredSocketCount()
   &SpawnRate(SpawnRate=15, Use Spawn Probability=false)
  ]
  particleSpawnScript [
   &VC_PaintedVariantsOverride(
    Black=<rgba 0.1198 0.1198 0.1198 1>
    BurntSienna=<rgba 0.7826 0.2416 0.0965 1>
    Cobalt=<rgba 0.0611 0.4802 0.8783 1>
    Crimson=<rgba 1 0.1217 0.1455 1>
    DefaultColor=&ColorFromCurve(ColorCurve=DataInterface() {propertyValues {
       AlphaCurve {keys [<at 0 1 RCIM_Cubic>,<at 0.6638 0 RCIM_Cubic>]}
       BlueCurve {keys [<at 0 1 RCIM_Cubic>]}
       GreenCurve {keys [<at 0 1 RCIM_Cubic>]}
       RedCurve {keys [<at 0 0.2086 RCIM_Cubic>]}
      }})
    ForestGreen=<rgba 0.0678 0.7652 0.103 1>
    Grey=<rgba 0.3177 0.3177 0.3177 1>
    Lime=<rgba 0.4252 0.7739 0.0766 1>
    Orange=<rgba 0.776 0.3154 0.0687 1>
    Pink=<rgba 0.6696 0.1106 0.4663 1>
    Purple=<rgba 0.3014 0.0954 0.8 1>
    Saffron=<rgba 0.9 0.7184 0.1219 1>
    SkyBlue=<rgba 0.0626 0.5533 0.8 1>
   )
   &InitializeParticle(
    Color Mode=::NewEnumerator0
    Lifetime=&MergeHelper_FloatSwitch(0=0.4, 1=0.5, 2=0.5, 3=?, Index=$User.GameModeID, Num=::NewEnumerator1)
    Position Mode=::NewEnumerator0
    Sprite Rotation Mode=::NewEnumerator2
    Sprite Size Mode=::NewEnumerator1
    Uniform Sprite Size Max=15
    Uniform Sprite Size Min=8
    Write Material Random=true
   )
   &PartitionParticles(Number Of Partitions=$Emitter.PortCount)
   &SkeletalMeshLocation(
    Bone / Socket Index=$Particles.PartitionParticles.Partition
    Bone / Socket Sampling Mode=::NewEnumerator5
    MaskSpawnGroup=false
    Mesh Sampling Type=::NewEnumerator6
    Sampled Position Offset=&AddVectorToPosition(Position=<0 0 0>, Vector=<-50 0 0>)
    Skeletal Mesh=$System.SkeletalMesh
   )
   &AddVelocity(Velocity=&MakeCustomVectorFromBool(
     Bool=$User.bIsFrontendPreview
     True Vector=&RandomRangeVector(Minimum=<-900 100 0>, Maximum=<-1150 -100 50>)
     False Vector=&RandomRangeVector(Minimum=<-900 100 0>, Maximum=<-1150 -100 50>)
    ))
   &ShapeLocation(MaskSpawnGroup=false, Override Seed=false, Sphere Radius=30, UseOwnerScale=true)
   /SetVariables(Particles.RendererTag:NiagaraInt32=&SpawnGroup()) {isSetParameters=true, module=SetVariables_#2}
  ]
  particleUpdateScript [
   &ParticleState()
   &InheritSourceMovement(
    Applied Position Scale=&Multiply_VectorByFloat(Vector=<1 1 1>, Float=1)
    ApplyForce=false
    ApplyPosition=true
    ApplyPositionFalloff=false
   )
   &SolveForcesAndVelocity(Clamp Velocity=false, Limit Acceleration=false)
  ]
  rendererClasses [&NiagaraSpriteRendererProperties]
  renderers [{rendererClass=&NiagaraSpriteRendererProperties}]
 }
 NE_Dots_Cone {
  emitterUpdateScript [
   &EmitterState()
   &VC_FilteredSocketCount()
   &SpawnRate(SpawnRate=20, Use Spawn Probability=false)
  ]
  particleSpawnScript [
   &VC_PaintedVariantsOverride(DefaultColor=&ColorFromCurve(ColorCurve=DataInterface() {propertyValues {
       AlphaCurve {keys [<at 0 1 RCIM_Cubic>,<at 0.6638 0 RCIM_Cubic>]}
       BlueCurve {keys [<at 0 1 RCIM_Cubic>]}
       GreenCurve {keys [<at 0 1 RCIM_Cubic>]}
       RedCurve {keys [<at 0 0.2086 RCIM_Cubic>]}
      }})) {enabled=false}
   &VC_PaintedVariantsOverride(
    Black=<rgba 0.1198 0.1198 0.1198 1>
    BurntSienna=<rgba 0.7826 0.2416 0.0965 1>
    Cobalt=<rgba 0.0611 0.4802 0.8783 1>
    Crimson=<rgba 1 0.1217 0.1455 1>
    DefaultColor=&ColorFromCurve(ColorCurve=DataInterface() {propertyValues {
       AlphaCurve {keys [<at 0 1 RCIM_Cubic>,<at 0.6638 0 RCIM_Cubic>]}
       BlueCurve {keys [<at 0 1 RCIM_Cubic>]}
       GreenCurve {keys [<at 0 1 RCIM_Cubic>]}
       RedCurve {keys [<at 0 0.2086 RCIM_Cubic>]}
      }})
    ForestGreen=<rgba 0.0678 0.7652 0.103 1>
    Grey=<rgba 0.3177 0.3177 0.3177 1>
    Lime=<rgba 0.4252 0.7739 0.0766 1>
    Orange=<rgba 0.776 0.3154 0.0687 1>
    Pink=<rgba 0.6696 0.1106 0.4663 1>
    Purple=<rgba 0.3014 0.0954 0.8 1>
    Saffron=<rgba 0.9 0.7184 0.1219 1>
    SkyBlue=<rgba 0.0626 0.5533 0.8 1>
   ) {module=VC_PaintedVariantsOverride001}
   &InitializeParticle(
    Color Mode=::NewEnumerator0
    Lifetime=&MergeHelper_FloatSwitch(0=0.1, 1=0.1, 2=0.1, 3=?, Index=$User.GameModeID, Num=::NewEnumerator1)
    Position Mode=::NewEnumerator0
    Sprite Rotation Mode=::NewEnumerator2
    Sprite Size Mode=::NewEnumerator1
    Uniform Sprite Size Max=15
    Uniform Sprite Size Min=8
    Write Material Random=true
   )
   &PartitionParticles(Number Of Partitions=$Emitter.PortCount)
   &SkeletalMeshLocation(
    Bone / Socket Index=$Particles.PartitionParticles.Partition
    Bone / Socket Sampling Mode=::NewEnumerator5
    MaskSpawnGroup=false
    Mesh Sampling Type=::NewEnumerator6
    Sampled Position Offset=&AddVectorToPosition(Position=<0 0 0>, Vector=<-50 0 0>)
    Skeletal Mesh=$System.SkeletalMesh
   )
   &AddVelocity(Velocity=&MakeCustomVectorFromBool(
     Bool=$User.bIsFrontendPreview
     True Vector=&RandomRangeVector(Minimum=<-900 200 400>, Maximum=<-900 -200 -400>)
     False Vector=&RandomRangeVector(Minimum=<-900 200 400>, Maximum=<-900 -200 -400>)
    ))
   &ShapeLocation(MaskSpawnGroup=false, Override Seed=false, Sphere Radius=35, UseOwnerScale=true)
   /SetVariables(Particles.RendererTag:NiagaraInt32=&SpawnGroup()) {isSetParameters=true, module=SetVariables_#2}
  ]
  particleUpdateScript [
   &ParticleState()
   &InheritSourceMovement(
    Applied Position Scale=&Multiply_VectorByFloat(Vector=<1 1 1>, Float=1)
    ApplyForce=false
    ApplyPosition=true
    ApplyPositionFalloff=false
   )
   &SolveForcesAndVelocity(Clamp Velocity=false, Limit Acceleration=false)
  ]
  rendererClasses [&NiagaraSpriteRendererProperties]
  renderers [{rendererClass=&NiagaraSpriteRendererProperties}]
 }
 NE_VC_Boost_Ribbons {
  emitterUpdateScript [
   &EmitterState(Inactive Response=::NewEnumerator1, Life Cycle Mode=::NewEnumerator1, Loop Duration=1, UseLoopDelay=false)
   &VC_FilteredSocketCount()
   &VC_RateQualityHandler(High Spawn Rate=60, Low Spawn Rate=60, Medium Spawn Rate=60, Use Custom Rate?=true)
   &VC_SpawnRibbon_v1(Number of Ribbons=$Emitter.PortCount, Spawn Time=-1, SpawnRate=$Emitter.Spawn Rate)
  ]
  particleSpawnScript [
   &VC_InitializeRibbon_v1(
    Custom Offset=true
    Enable Skeletal Position=true
    Facing Axis:ENiagaraOrientationAxis=::YAxis
    Lifetime=&MergeHelper_FloatSwitch(0=1, 1=1, 2=1, 3=?, Index=$User.GameModeID, Num=::NewEnumerator1)
    Local Vector Offset=&AddVectorToPosition(Position=<-5 0 -8>, Vector=$User.DTCarOffset)
    Offset Amount=1
    Ribbon Width=90
   )
   &AddVelocity(Velocity=&MaskVector(Mask Source=1, Mask Parameter=$Particles.SpawnGroup, Mask=0, Masked Vector=<-1000 100 50>, Passthrough Vector=<-1000 -100 50>)) {module=AddVelocity001}
   &PartitionParticles(Number Of Partitions=$Emitter.PortCount)
  ]
  particleUpdateScript [
   &VC_PaintedVariantsOverride(
    Orange=<rgba 0.6724 0.0883 0.2836 1>
    R1 Black:LinearColor=&ColorFromCurve(ColorCurve=DataInterface() {propertyValues {
       AlphaCurve {keys [<at 0 0.2 RCIM_Cubic>,<at 0.1288 1 RCIM_Cubic>]}
       BlueCurve {keys [<at 0 0.013 RCIM_Cubic>]}
       GreenCurve {keys [<at 0 0.016 RCIM_Cubic>]}
       RedCurve {keys [<at 0 0.016 RCIM_Cubic>]}
      }})
    R1 BurntSienna:LinearColor=&ColorFromCurve(ColorCurve=DataInterface() {propertyValues {
       AlphaCurve {keys [<at 0 0.2 RCIM_Cubic>,<at 0.1288 1 RCIM_Cubic>]}
       BlueCurve {keys [<at 0 0.0082 RCIM_Cubic>]}
       GreenCurve {keys [<at 0 0.0231 RCIM_Cubic>]}
       RedCurve {keys [<at 0 0.0833 RCIM_Cubic>]}
      }})
    R1 Cobalt:LinearColor=&ColorFromCurve(ColorCurve=DataInterface() {propertyValues {
       AlphaCurve {keys [<at 0 0.2 RCIM_Cubic>,<at 0.1288 1 RCIM_Cubic>]}
       BlueCurve {keys [<at 0 0.2448 RCIM_Cubic>]}
       GreenCurve {keys [<at 0 0.0442 RCIM_Cubic>]}
       RedCurve {keys [<at 0 0.0054 RCIM_Cubic>]}
      }})
    R1 Crimson:LinearColor=&ColorFromCurve(ColorCurve=DataInterface() {propertyValues {
       AlphaCurve {keys [<at 0 0.2 RCIM_Cubic>,<at 0.1288 1 RCIM_Cubic>]}
       BlueCurve {keys [<at 0 0.015 RCIM_Cubic>]}
       GreenCurve {keys [<at 0 0.0222 RCIM_Cubic>]}
       RedCurve {keys [<at 0 0.2812 RCIM_Cubic>]}
      }})
    R1 DefaultColor=&ColorFromCurve(ColorCurve=DataInterface() {propertyValues {
       AlphaCurve {keys [<at 0 0.2 RCIM_Cubic>,<at 0.1288 1 RCIM_Cubic>]}
       BlueCurve {keys [<at 0 0.0782 RCIM_Cubic>]}
       GreenCurve {keys [<at 0 0.0033 RCIM_Cubic>]}
       RedCurve {keys [<at 0 1 RCIM_Cubic>]}
      }})
    R1 ForestGreen:LinearColor=&ColorFromCurve(ColorCurve=DataInterface() {propertyValues {
       AlphaCurve {keys [<at 0 0.2 RCIM_Cubic>,<at 0.1288 1 RCIM_Cubic>]}
       BlueCurve {keys [<at 0 0.0065 RCIM_Cubic>]}
       GreenCurve {keys [<at 0 0.1302 RCIM_Cubic>]}
       RedCurve {keys [<at 0 0 RCIM_Cubic>]}
      }})
    R1 Grey:LinearColor=&ColorFromCurve(ColorCurve=DataInterface() {propertyValues {
       AlphaCurve {keys [<at 0 0.2 RCIM_Cubic>,<at 0.1288 1 RCIM_Cubic>]}
       BlueCurve {keys [<at 0 0.0833 RCIM_Cubic>]}
       GreenCurve {keys [<at 0 0.0833 RCIM_Cubic>]}
       RedCurve {keys [<at 0 0.0833 RCIM_Cubic>]}
      }})
    R1 Lime:LinearColor=&ColorFromCurve(ColorCurve=DataInterface() {propertyValues {
       AlphaCurve {keys [<at 0 0.2 RCIM_Cubic>,<at 0.1288 1 RCIM_Cubic>]}
       BlueCurve {keys [<at 0 0.0085 RCIM_Cubic>]}
       GreenCurve {keys [<at 0 0.1391 RCIM_Cubic>]}
       RedCurve {keys [<at 0 0.074 RCIM_Cubic>]}
      }})
    R1 Orange:LinearColor=&ColorFromCurve(ColorCurve=DataInterface() {propertyValues {
       AlphaCurve {keys [<at 0 0.2 RCIM_Cubic>,<at 0.1288 1 RCIM_Cubic>]}
       BlueCurve {keys [<at 0 0.0059 RCIM_Cubic>]}
       GreenCurve {keys [<at 0 0.0478 RCIM_Cubic>]}
       RedCurve {keys [<at 0 0.2292 RCIM_Cubic>]}
      }})
    R1 Pink:LinearColor=&ColorFromCurve(ColorCurve=DataInterface() {propertyValues {
       AlphaCurve {keys [<at 0 0.2 RCIM_Cubic>,<at 0.1288 1 RCIM_Cubic>]}
       BlueCurve {keys [<at 0 0.1735 RCIM_Cubic>]}
       GreenCurve {keys [<at 0 0.0311 RCIM_Cubic>]}
       RedCurve {keys [<at 0 0.2552 RCIM_Cubic>]}
      }})
    R1 Purple:LinearColor=&ColorFromCurve(ColorCurve=DataInterface() {propertyValues {
       AlphaCurve {keys [<at 0 0.2 RCIM_Cubic>,<at 0.1288 1 RCIM_Cubic>]}
       BlueCurve {keys [<at 0 0.3229 RCIM_Cubic>]}
       GreenCurve {keys [<at 0 0.0104 RCIM_Cubic>]}
       RedCurve {keys [<at 0 0.0951 RCIM_Cubic>]}
      }})
    R1 Saffron:LinearColor=&ColorFromCurve(ColorCurve=DataInterface() {propertyValues {
       AlphaCurve {keys [<at 0 0.2 RCIM_Cubic>,<at 0.1288 1 RCIM_Cubic>]}
       BlueCurve {keys [<at 0 0 RCIM_Cubic>]}
       GreenCurve {keys [<at 0 0.1796 RCIM_Cubic>]}
       RedCurve {keys [<at 0 0.2344 RCIM_Cubic>]}
      }})
    R1 SkyBlue:LinearColor=&ColorFromCurve(ColorCurve=DataInterface() {propertyValues {
       AlphaCurve {keys [<at 0 0.2 RCIM_Cubic>,<at 0.1288 1 RCIM_Cubic>]}
       BlueCurve {keys [<at 0 0.1927 RCIM_Cubic>]}
       GreenCurve {keys [<at 0 0.1288 RCIM_Cubic>]}
       RedCurve {keys [<at 0 0 RCIM_Cubic>]}
      }})
    R1 TitaniumWhite:LinearColor=&ColorFromCurve(ColorCurve=DataInterface() {propertyValues {
       AlphaCurve {keys [<at 0 0.2 RCIM_Cubic>,<at 0.1288 1 RCIM_Cubic>]}
       BlueCurve {keys [<at 0 0.2552 RCIM_Cubic>]}
       GreenCurve {keys [<at 0 0.2552 RCIM_Cubic>]}
       RedCurve {keys [<at 0 0.2552 RCIM_Cubic>]}
      }})
    R2 Black:LinearColor=&ColorFromCurve(ColorCurve=DataInterface() {propertyValues {
       AlphaCurve {keys [<at 0 0.2 RCIM_Cubic>,<at 0.0968 1 RCIM_Cubic>]}
       BlueCurve {keys [<at 0 0.0546 RCIM_Cubic>]}
       GreenCurve {keys [<at 0 0.0573 RCIM_Cubic>]}
       RedCurve {keys [<at 0 0.0573 RCIM_Cubic>]}
      }})
    R2 BurntSienna:LinearColor=&ColorFromCurve(ColorCurve=DataInterface() {propertyValues {
       AlphaCurve {keys [<at 0 0.2 RCIM_Cubic>,<at 0.0968 1 RCIM_Cubic>]}
       BlueCurve {keys [<at 0 0.0938 RCIM_Cubic>]}
       GreenCurve {keys [<at 0 0.1204 RCIM_Cubic>]}
       RedCurve {keys [<at 0 0.2604 RCIM_Cubic>]}
      }})
    R2 Cobalt:LinearColor=&ColorFromCurve(ColorCurve=DataInterface() {propertyValues {
       AlphaCurve {keys [<at 0 0.2 RCIM_Cubic>,<at 0.0968 1 RCIM_Cubic>]}
       BlueCurve {keys [<at 0 0.4696 RCIM_Cubic>]}
       GreenCurve {keys [<at 0 0.1884 RCIM_Cubic>]}
       RedCurve {keys [<at 0 0.0354 RCIM_Cubic>]}
      }})
    R2 Crimson:LinearColor=&ColorFromCurve(ColorCurve=DataInterface() {propertyValues {
       AlphaCurve {keys [<at 0 0.2 RCIM_Cubic>,<at 0.0968 1 RCIM_Cubic>]}
       BlueCurve {keys [<at 0 0.0839 RCIM_Cubic>]}
       GreenCurve {keys [<at 0 0.0717 RCIM_Cubic>]}
       RedCurve {keys [<at 0 0.687 RCIM_Cubic>]}
      }})
    R2 DefaultColor:LinearColor=&ColorFromCurve(ColorCurve=DataInterface() {propertyValues {
       AlphaCurve {keys [<at 0 0.2 RCIM_Cubic>,<at 0.0968 1 RCIM_Cubic>]}
       BlueCurve {keys [<at 0 0.0021 RCIM_Cubic>]}
       GreenCurve {keys [<at 0 0.314 RCIM_Cubic>]}
       RedCurve {keys [<at 0 0.9734 RCIM_Cubic>]}
      }})
    R2 ForestGreen:LinearColor=&ColorFromCurve(ColorCurve=DataInterface() {propertyValues {
       AlphaCurve {keys [<at 0 0.2 RCIM_Cubic>,<at 0.0968 1 RCIM_Cubic>]}
       BlueCurve {keys [<at 0 0.0274 RCIM_Cubic>]}
       GreenCurve {keys [<at 0 0.4087 RCIM_Cubic>]}
       RedCurve {keys [<at 0 0.0368 RCIM_Cubic>]}
      }})
    R2 Grey:LinearColor=&ColorFromCurve(ColorCurve=DataInterface() {propertyValues {
       AlphaCurve {keys [<at 0 0.2 RCIM_Cubic>,<at 0.0968 1 RCIM_Cubic>]}
       BlueCurve {keys [<at 0 0.17 RCIM_Cubic>]}
       GreenCurve {keys [<at 0 0.17 RCIM_Cubic>]}
       RedCurve {keys [<at 0 0.1739 RCIM_Cubic>]}
      }})
    R2 Lime:LinearColor=&ColorFromCurve(ColorCurve=DataInterface() {propertyValues {
       AlphaCurve {keys [<at 0 0.2 RCIM_Cubic>,<at 0.0968 1 RCIM_Cubic>]}
       BlueCurve {keys [<at 0 0.0266 RCIM_Cubic>]}
       GreenCurve {keys [<at 0 0.3826 RCIM_Cubic>]}
       RedCurve {keys [<at 0 0.2945 RCIM_Cubic>]}
      }})
    R2 Orange:LinearColor=&ColorFromCurve(ColorCurve=DataInterface() {propertyValues {
       AlphaCurve {keys [<at 0 0.2 RCIM_Cubic>,<at 0.0968 1 RCIM_Cubic>]}
       BlueCurve {keys [<at 0 0.0179 RCIM_Cubic>]}
       GreenCurve {keys [<at 0 0.1768 RCIM_Cubic>]}
       RedCurve {keys [<at 0 0.3826 RCIM_Cubic>]}
      }})
    R2 Pink:LinearColor=&ColorFromCurve(ColorCurve=DataInterface() {propertyValues {
       AlphaCurve {keys [<at 0 0.2 RCIM_Cubic>,<at 0.0968 1 RCIM_Cubic>]}
       BlueCurve {keys [<at 0 0.4014 RCIM_Cubic>]}
       GreenCurve {keys [<at 0 0.0882 RCIM_Cubic>]}
       RedCurve {keys [<at 0 0.4087 RCIM_Cubic>]}
      }})
    R2 Purple:LinearColor=&ColorFromCurve(ColorCurve=DataInterface() {propertyValues {
       AlphaCurve {keys [<at 0 0.2 RCIM_Cubic>,<at 0.0968 1 RCIM_Cubic>]}
       BlueCurve {keys [<at 0 0.3229 RCIM_Cubic>]}
       GreenCurve {keys [<at 0 0.0562 RCIM_Cubic>]}
       RedCurve {keys [<at 0 0.2769 RCIM_Cubic>]}
      }})
    R2 Saffron:LinearColor=&ColorFromCurve(ColorCurve=DataInterface() {propertyValues {
       AlphaCurve {keys [<at 0 0.2 RCIM_Cubic>,<at 0.0968 1 RCIM_Cubic>]}
       BlueCurve {keys [<at 0 0.0681 RCIM_Cubic>]}
       GreenCurve {keys [<at 0 0.4276 RCIM_Cubic>]}
       RedCurve {keys [<at 0 0.4348 RCIM_Cubic>]}
      }})
    R2 SkyBlue:LinearColor=&ColorFromCurve(ColorCurve=DataInterface() {propertyValues {
       AlphaCurve {keys [<at 0 0.2 RCIM_Cubic>,<at 0.0968 1 RCIM_Cubic>]}
       BlueCurve {keys [<at 0 0.3739 RCIM_Cubic>]}
       GreenCurve {keys [<at 0 0.3369 RCIM_Cubic>]}
       RedCurve {keys [<at 0 0.0533 RCIM_Cubic>]}
      }})
    R2 TitaniumWhite:LinearColor=&ColorFromCurve(ColorCurve=DataInterface() {propertyValues {
       AlphaCurve {keys [<at 0 0.2 RCIM_Cubic>,<at 0.0968 1 RCIM_Cubic>]}
       BlueCurve {keys [<at 0 0.9647 RCIM_Cubic>]}
       GreenCurve {keys [<at 0 0.9823 RCIM_Cubic>]}
       RedCurve {keys [<at 0 0.9823 RCIM_Cubic>]}
      }})
    R3 Black:LinearColor=<rgba 0.1094 0.1094 0.1094 1>
    R3 BurntSienna:LinearColor=<rgba 0.2664 0.1384 0.0844 1>
    R3 Cobalt:LinearColor=<rgba 0.0488 0.0637 0.8522 1>
    R3 Crimson:LinearColor=<rgba 0.4635 0.0748 0.0854 1>
    R3 DefaultColor:LinearColor=&ColorFromCurve(ColorCurve=DataInterface() {propertyValues {
       AlphaCurve {keys [<at 0 1 RCIM_Cubic>]}
       BlueCurve {keys [<at 0.2416 0.314 RCIM_Cubic>,<at 0.6268 0.9823 RCIM_Cubic>]}
       GreenCurve {keys [<at 0.2416 0.0382 RCIM_Cubic>,<at 0.6268 0.9734 RCIM_Cubic>]}
       RedCurve {keys [<at 0.2416 0.6654 RCIM_Cubic>,<at 0.6268 0.0482 RCIM_Cubic>]}
      }})
    R3 ForestGreen:LinearColor=<rgba 0.1329 0.3594 0.1443 1>
    R3 Grey:LinearColor=<rgba 0.4323 0.4323 0.4323 1>
    R3 Lime:LinearColor=<rgba 0.3387 0.5885 0.0889 1>
    R3 Orange:LinearColor=<rgba 0.6146 0.2477 0.0512 1>
    R3 Pink:LinearColor=<rgba 0.5052 0.1421 0.3732 1>
    R3 Purple:LinearColor=<rgba 0.329 0.0849 0.6795 1>
    R3 Saffron:LinearColor=<rgba 0.9 0.713 0.0984 1>
    R3 SkyBlue:LinearColor=<rgba 0.1181 0.5108 0.7083 1>
    R3 TitaniumWhite:LinearColor=<rgba 0.8906 0.8906 0.8906 1>
    Use R1 Painted Overrides=true
    Use R2 Painted Overrides=true
    Use R3 Painted Overrides=true
   )
   &ParticleState()
   &DynamicMaterialParameters(
    Index 2 Param 1=1
    Index 2 Param 2=&MaskFloat(Mask Source=1, Mask Parameter=$Particles.PartitionParticles.Partition, Mask=0, Masked Float=1, Passthrough Float=-1)
    Index 2 Param 3=1
    Index 2 Param 4=&MaskFloat(Mask Source=1, Mask Parameter=$Particles.PartitionParticles.Partition, Mask=0, Masked Float=1, Passthrough Float=-1)
    Module.Index0Float0Write=true
    Module.Index0Float1Write=true
    Module.Index0Float2Write=true
    Module.Index0Float3Write=true
    Module.Index1Float0Write=true
    Module.Index1Float1Write=true
    Module.Index1Float2Write=true
    Module.Index1Float3Write=true
    Module.Index2Float0Write=true
    Module.Index2Float1Write=true
    Module.Index2Float2Write=true
    Module.Index2Float3Write=true
    Write Parameter Index 1=::NewEnumerator1
    Write Parameter Index 2=::NewEnumerator1
   )
   &InheritSourceMovement(ApplyForce=false, ApplyPosition=true, ApplyPositionFalloff=false) {module=InheritSourceMovement002}
   &SolveForcesAndVelocity(Clamp Velocity=false, Limit Acceleration=false)
   &VC_UpdateRibbon_v1(Enable Skeletal Position=true)
   /SetVariables(
    Particles.R1_Width:NiagaraFloat=&FloatFromCurve(
     FloatCurve=DataInterface() {propertyValues {Curve {keys [
         <at -0.0033 0.4028 RCIM_Cubic>
         {arriveTangent=2.6397, interpMode=RCIM_Cubic, leaveTangent=2.6397, time=0.2152, value=1.5181}
         {arriveTangent=-0.7845, interpMode=RCIM_Cubic, leaveTangent=-0.7845, tangentMode=RCTM_User, time=0.4646, value=1.6379}
         {interpMode=RCIM_Cubic, tangentMode=RCTM_Break, time=1.0058, value=0.0994}
        ]}}}
     Scale Curve=190
    )
    Particles.R2_Width:NiagaraFloat=&FloatFromCurve(
     FloatCurve=DataInterface() {propertyValues {Curve {keys [
         <at -0.0033 0.4028 RCIM_Cubic>
         {arriveTangent=2.6397, interpMode=RCIM_Cubic, leaveTangent=2.6397, time=0.2152, value=1.5181}
         {arriveTangent=-0.7845, interpMode=RCIM_Cubic, leaveTangent=-0.7845, tangentMode=RCTM_User, time=0.4646, value=1.6379}
         {interpMode=RCIM_Cubic, tangentMode=RCTM_Break, time=1.0058, value=0.0994}
        ]}}}
     Scale Curve=100
    )
    Particles.R4_Width:NiagaraFloat=&FloatFromCurve(
     FloatCurve=DataInterface() {propertyValues {Curve {keys [
         <at -0.0033 0.4028 RCIM_Cubic>
         {arriveTangent=2.6397, interpMode=RCIM_Cubic, leaveTangent=2.6397, time=0.2152, value=1.5181}
         {arriveTangent=-0.7845, interpMode=RCIM_Cubic, leaveTangent=-0.7845, tangentMode=RCTM_User, time=0.4646, value=1.6379}
         {interpMode=RCIM_Cubic, tangentMode=RCTM_Break, time=1.0058, value=0.0994}
        ]}}}
     Scale Curve=194
    )
   ) {isSetParameters=true}
  ]
  rendererClasses [&NiagaraRibbonRendererProperties]
  renderers [
   {rendererClass=&NiagaraRibbonRendererProperties}
   {rendererClass=&NiagaraRibbonRendererProperties, rendererIndex=1}
   {rendererClass=&NiagaraRibbonRendererProperties, rendererIndex=2}
   {rendererClass=&NiagaraRibbonRendererProperties, rendererIndex=3}
   {rendererClass=&NiagaraRibbonRendererProperties, rendererIndex=4}
  ]
 }
}
summary {
 systemName=NS_Boost_StreetArt
 emitters [
  {emitterName=NE_Dots, rendererClasses [&NiagaraSpriteRendererProperties]}
  {emitterName=NE_VC_Boost_Ribbons, rendererClasses [&NiagaraRibbonRendererProperties]}
  {emitterName=NE_Dots_Cone, rendererClasses [&NiagaraSpriteRendererProperties]}
 ]
 userVariables [
  User.bIsFrontendPreview {type=NiagaraBool}
  User.bIsPainted {type=NiagaraBool}
  User.ColorVariantID {type=NiagaraInt32}
  User.DTCarOffset {type=Vector3f, description="", defaultValue {struct=Vector3f, value=<0 0 0>}}
  User.GameModeID {type=NiagaraInt32}
  User.PaintedColor {type=LinearColor, description="", defaultValue {struct=LinearColor, value=<rgba 1 1 1 1>}}
  User.PaintedScalar {type=NiagaraFloat, defaultValue=1, description=""}
  User.PortCount {type=NiagaraInt32, defaultValue=1}
 ]
}
